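# M_DeimosPaintMask is a Default Lit material for skeletal meshes and mesh deformers that displays a Deimos FX mask texture as emissive color.
# A SetMaterialAttributes node feeds the RGB of the Elite mask sample into Emissive Color over a constant black Base Color.
# A second sample of the Fiend mask texture sits in the graph unconnected.
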Material /PulseLancer_SmashAI/Assets/Textures/TA/M_DeimosPaintMask
graph {
 properties {
  bAutomaticallySetUsageInEditor=false
  bUseMaterialAttributes=true
  bUsedWithMeshDeformer=true
  bUsedWithSkeletalMesh=true
  usedShadingModels=Default Lit
 }
 nodes [
  SetMaterialAttributes!(
   MaterialAttributes
   Base Color=Constant()
   Metallic
   Specular
   Roughness
   Anisotropy
   Emissive Color=TextureSample() {samplerType=SAMPLERTYPE_Masks, texture=/PulseLancer_SmashAI/Assets/Textures/T_DeimosFXMasks_Elite}.RGB
   Opacity
   Opacity Mask
   Normal
   Tangent
  ) {attributeSetTypes [
    #1
    #2
    #3
    #4
    #5
    #6
    #7
    #8
    #9
    #10
   ]}
  TextureSample() {samplerType=SAMPLERTYPE_Masks, texture=/PulseLancer_SmashAI/Assets/Textures/T_DeimosFXMasks_Fiend}
 ]
}
